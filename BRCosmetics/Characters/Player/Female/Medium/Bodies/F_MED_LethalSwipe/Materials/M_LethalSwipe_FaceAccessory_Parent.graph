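# Parent material for the LethalSwipe female character's face accessory, built on the 2023 character parent functions.
# The PreFX output is blended with FN hair shading through a hair mask, with static switches for that hair path and optional flyaway hair.
# A colored Fresnel glow, masked by the Tech Art Mask green channel and tinted bluer at night via time-of-day color, adds to emissive before PostFX.
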Material /BRCosmetics/Characters/Player/Female/Medium/Bodies/F_MED_LethalSwipe/Materials/M_LethalSwipe_FaceAccessory_Parent
graph {
 properties {
  bUseMaterialAttributes=true
  bUsedWithInstancedStaticMeshes=true
  bUsedWithMeshDeformer=true
  bUsedWithMorphTargets=true
  bUsedWithSkeletalMesh=true
  pixelDepthOffsetMode=PDOM_Legacy
  shadingModel=MSM_FromMaterialExpression
  usedShadingModels=Default Lit | Subsurface | Clear Coat | Subsurface Profile | Cloth
 }
 nodes [
  n1 = MaterialFunctionCall() {materialFunction=&MF_CharacterParent_2023_PreFX}
  n2 = MaterialFunctionCall!(In=@n1.Result, Mask Input=@n1.Result) {materialFunction=/Game/Global/Materials/Character_Parent_2023/Hair_Parent_2023/MaterialFunctions/MF_FN_Hair_2023}
  n3 = BlendMaterialAttributes(
   A=@n1.Result
   B=StaticSwitchParameter(True=@n2.Result, False=@n1.Result) {defaultValue=true, group=Skin FX, parameterName=UseFNHair2_0}
   Alpha=@n2.HairMask
  )
  n4 = GetMaterialAttributes(None=StaticSwitchParameter(
    True=MaterialFunctionCall!(MaterialInput=@n3) {materialFunction=/Game/Global/Materials/Character_Parent_2023/Hair_Parent_2023/MaterialFunctions/MF_FlyAwayHair_2023}.Result
    False=@n3
   ) {group=Skin FX, parameterName=IsFlyAwayHair?}) {attributeGetTypes [#1]}
  n5 = Fresnel(ExponentIn=ScalarParameter() {defaultValue=3, parameterName=Fresnel Exp})
  n6 = VectorParameter() {parameterName=Fresnel Color, defaultValue=<rgba 1 0.8333 0 1>}
  n7 = Constant3Vector() {constant=<rgba 1 1 1 0>}
  n8 = Multiply(
   A=Multiply(A=ComponentMask(None=@n1.Tech Art Mask) {g=true}, B=Multiply(A=@n5, B=@n6.RGB))
   B=MaterialFunctionCall(Color_Morning=@n7, Color_Day=@n7, Color_Evening=@n7, Color_Night=Constant3Vector() {constant=<rgba 0.5 0.5833 1 0>}) {materialFunction=&Color_Time_of_Day}.Result
  )
  MaterialFunctionCall(
   Material=SetMaterialAttributes(MaterialAttributes=@n4.MaterialAttributes, Emissive Color=Add(A=@n4.Emissive Color, B=@n8)) {attributeSetTypes [#1]}
   M=@n1.M
   Pre Skinned Local Position=@n1.M
   Distance Blend=@n1.M
   Fresnel=@n1.M
  ) {materialFunction=&MF_CharacterParent_2023_PostFX}
 ]
}
