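# Parent character material for the Lethal Swipe skin, adding a slurp and parallax-bubble effect to the shared 2023 character shader.
# Chains PreFX, UV2 composite and Slurp functions, then mask-blends custom base color and emissive before PostFX.
# Custom layers combine bubbles, an edge Fresnel glow and dark green shading, with time-of-day tinting on emissive.
# A static switch selects Fresnel or AO-style darkening for the darkening mask.
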Material /BRCosmetics/Characters/Player/Female/Medium/Bodies/F_MED_LethalSwipe/Materials/M_LethalSwipe_Parent
graph {
 properties {
  bAutomaticallySetUsageInEditor=false
  bUseMaterialAttributes=true
  bUsedWithMeshDeformer=true
  bUsedWithMorphTargets=true
  bUsedWithSkeletalMesh=true
  shadingModel=MSM_FromMaterialExpression
  usedShadingModels=Default Lit | Subsurface | Clear Coat | Subsurface Profile | Cloth
 }
 nodes [
  n1 = MaterialFunctionCall() {materialFunction=&MF_CharacterParent_2023_PreFX}
  n2 = GetMaterialAttributes(None=MaterialFunctionCall!(
    SlurpMask=NamedRerouteUsage()
    AO
    BubblesMask
    SlurpNormalDistortionMask
    Pre Skinned Position
    Material=MaterialFunctionCall!(Material=@n1.Result, TechArtMask=@n1.Result) {materialFunction=/Game/Global/Materials/Character_Parent_2023/UV2Composite_Parent_2023/MF_UV2_Composite_2023}.Result
   ) {materialFunction=/Game/Global/Materials/Character_Parent_2023/Slurp_Parent_2023/MF_Slurp_2023}.Result) {attributeGetTypes [#1, #2]}
  n3 = NamedRerouteUsage()
  n4 = Constant4Vector() {constant=<rgba 0 0.1 0 1>}
  n5 = NamedRerouteUsage()
  n6 = MaterialFunctionCall!(Background Parallax Offset, Mask=NamedRerouteUsage(), Foreground Parallax Offset) {materialFunction=&/MF_LethalSwipe_Bubbles}
  n7 = LinearInterpolate(A=@n4.RGB, B=LinearInterpolate(A=@n2.Base Color, B=@n6.Result, Alpha=@n6.BubblesMask), Alpha=@n5)
  n8 = NamedRerouteUsage()
  n9 = NamedRerouteUsage()
  n10 = Add(A=@n2.Emissive Color, B=Multiply(A=@n6.Result, B=@n6.Result))
  n11 = Multiply(
   A=Fresnel(ExponentIn=ScalarParameter() {defaultValue=3, group=Edge Fresnel, parameterName=Edge Fresnel Exp})
   B=VectorParameter() {group=Edge Fresnel, parameterName=Edge Fresnel Color, defaultValue=<rgba 1 0.8333 0 1>}.RGB
  )
  n12 = Constant3Vector() {constant=<rgba 1 1 1 0>}
  n13 = Multiply(
   A=LinearInterpolate(A=@n4.RGB, B=Add(A=@n10, B=@n11), Alpha=@n5)
   B=MaterialFunctionCall(Color_Morning=@n12, Color_Day=@n12, Color_Evening=@n12, Color_Night=Constant3Vector() {constant=<rgba 0.5 0.5833 1 0>}) {materialFunction=&Color_Time_of_Day}.Result
  )
  MaterialFunctionCall(
   Material=SetMaterialAttributes!(
    MaterialAttributes=@n2.MaterialAttributes
    Base Color=LinearInterpolate(A=@n3, B=@n7, Alpha=@n8)
    Emissive Color=LinearInterpolate(A=@n9, B=@n13, Alpha=@n8)
   ) {attributeSetTypes [#1, #2]}
   M=@n1.M
   Pre Skinned Local Position=@n1.M
   Distance Blend=@n1.M
   Fresnel=@n1.M
  ) {materialFunction=&MF_CharacterParent_2023_PostFX}
  NamedRerouteDeclaration(None=ComponentMask(None=@n1.M) {r=true}) {name=AO mask, nodeColor=<rgba 0 0.8125 1 1>}
  NamedRerouteDeclaration(None=ComponentMask(None=@n1.Tech Art Mask) {g=true}) {name=SkinMask, nodeColor=<rgba 0.7344 1 0 1>}
  NamedRerouteDeclaration(None=@n2.Base Color) {name=InBC, nodeColor=<rgba 0 1 0.7188 1>}
  NamedRerouteDeclaration(None=@n2.Emissive Color) {name=InEmissive, nodeColor=<rgba 1 0.3984 0 1>}
  n14 = Add(
   A=Fresnel(ExponentIn=ScalarParameter() {defaultValue=1, group=Darkening, parameterName=Darkening_FresnelExp})
   B=OneMinus(None=ScalarParameter() {defaultValue=0.5, group=Darkening, parameterName=Darkening_FresnelIntensity})
  )
  n15 = NamedRerouteUsage()
  n16 = Constant() {r=0.6}
  NamedRerouteDeclaration(None=StaticSwitchParameter(True=Saturate(None=@n14), False=Power(Base=@n15, Exp=@n16)) {group=Darkening, parameterName=Fresnel/AO_Darkening}) {name=DarkeningMask, nodeColor=<rgba 0.7578 0 1 1>}
 ]
}
