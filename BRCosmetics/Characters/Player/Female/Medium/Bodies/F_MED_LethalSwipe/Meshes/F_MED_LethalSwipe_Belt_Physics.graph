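# Physics asset for the LethalSwipe belt on the medium female player body, setting up jointed capsule bodies for belt movement.
# Pelvis, both thighs and the left calf carry body capsules joined by angle-locked constraints.
# Back and front belt chains start angle-locked to the pelvis through base bones, then continue with small swing limits.
# A free-rotation constraint joins the last back and front links, closing the chains into a loop.
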PhysicsAsset /BRCosmetics/Characters/Player/Female/Medium/Bodies/F_MED_LethalSwipe/Meshes/F_MED_LethalSwipe_Belt_Physics
bodies {
 base_belt_B_01 [{shapeType=Capsule, length=0.505, radius=0.505, shapeName=base_belt_B_01_capsule, rotation=<rot 0 -90 0>}]
 base_belt_F_01 [{shapeType=Capsule, length=0.505, radius=0.505, shapeName=base_belt_F_01_capsule, rotation=<rot 0 -90 0>}]
 calf_l [{shapeType=Capsule, length=28.1274, radius=12.1308, center=<-28.2813 -0.2689 -0.7494>, rotation=<rot 0 78.304 -88.5873>}]
 dyn_belt_B_01 [{
   shapeType=Capsule
   length=8.0552
   radius=2.5912
   shapeName=dyn_belt_B_01_capsule
   center=<-0.6641 2.4452 0.669>
   rotation=<rot 0 -128.4094 -80.1544>
  }]
 dyn_belt_B_02 [{shapeType=Capsule, length=7.5, radius=2, shapeName=dyn_belt_B_02_capsule, center=<6.1635 -1.0666 0>, rotation=<rot 90 0 0>}]
 dyn_belt_B_03 [{shapeType=Capsule, length=2.5, radius=2, shapeName=dyn_belt_B_03_capsule, center=<3.4626 -1.8344 0>, rotation=<rot 90 0 0>}]
 dyn_belt_B_04 [{shapeType=Capsule, length=3, radius=1.5, shapeName=dyn_belt_B_04_capsule, center=<2.8567 -0.963 0.7072>, rotation=<rot 90 0 0>}]
 dyn_belt_B_05 [
  {shapeType=Capsule, length=3, radius=1.5, shapeName=dyn_belt_B_05_capsule, center=<0.9773 -1.3864 1.2>, rotation=<rot 90 -5 0>}
  {
   shapeType=Capsule
   length=10.5
   radius=4
   shapeName=dyn_belt_B_05_capsule
   center=<-1.7897 -5.2133 2.8864>
   rotation=<rot 4.9236 -0.8704 -10.0375>
  }
 ]
 dyn_belt_F_01 [{shapeType=Capsule, length=3, radius=1.5, shapeName=dyn_belt_F_01_capsule, center=<3.6 -1.3089 0>, rotation=<rot 90 -0.0003 0>}]
 dyn_belt_F_02 [{shapeType=Capsule, length=3, radius=1.5, shapeName=dyn_belt_F_02_capsule, center=<2 -1.5 0>, rotation=<rot 90 0 0>}]
 dyn_belt_F_03 [{shapeType=Capsule, length=3, radius=1.5, shapeName=dyn_belt_F_03_capsule, center=<2 -1.5 0>, rotation=<rot 90 0 0>}]
 dyn_belt_F_04 [{shapeType=Capsule, length=3, radius=1.5, shapeName=dyn_belt_F_04_capsule, center=<2 -1.1774 -1.182>, rotation=<rot 90 0 0>}]
 pelvis [
  {shapeType=Capsule, length=6.9051, radius=11, center=<-0.864 -1.864 0>}
  {shapeType=Capsule, length=10.5, radius=10, shapeName=pelvis_capsule, center=<-3.5871 -3.5415 0>}
 ]
 thigh_l [{shapeType=Capsule, length=21.902, radius=8.4023, center=<-15.4094 -0.2135 0.45>, rotation=<rot -0.0001 86.3905 -83.781>}]
 thigh_r [{shapeType=Capsule, length=21.402, radius=8.9023, center=<15.3767 -0.2314 -0.8831>, rotation=<rot -0.0001 86.3903 96.219>}]
}
constraints [
 {
  bone1Name=thigh_l
  bone2Name=pelvis
  swing1LimitDegrees=45
  swing1Motion=Locked
  swing2LimitDegrees=45
  swing2Motion=Locked
  twistLimitDegrees=45
  twistMotion=Locked
 }
 {
  bone1Name=calf_l
  bone2Name=thigh_l
  swing1LimitDegrees=45
  swing1Motion=Locked
  swing2LimitDegrees=45
  swing2Motion=Locked
  twistLimitDegrees=45
  twistMotion=Locked
 }
 {
  bone1Name=thigh_r
  bone2Name=pelvis
  swing1LimitDegrees=45
  swing1Motion=Locked
  swing2LimitDegrees=45
  swing2Motion=Locked
  twistLimitDegrees=45
  twistMotion=Locked
 }
 {
  bone1Name=base_belt_B_01
  bone2Name=pelvis
  swing1LimitDegrees=45
  swing1Motion=Locked
  swing2LimitDegrees=45
  swing2Motion=Locked
  twistLimitDegrees=45
  twistMotion=Locked
 }
 {
  bone1Name=dyn_belt_B_01
  bone2Name=base_belt_B_01
  swing1LimitDegrees=45
  swing1Motion=Locked
  swing2LimitDegrees=45
  swing2Motion=Locked
  twistLimitDegrees=45
  twistMotion=Locked
 }
 {bone1Name=dyn_belt_B_02, bone2Name=dyn_belt_B_01, swing1LimitDegrees=5, swing2LimitDegrees=10, twistLimitDegrees=0}
 {bone1Name=dyn_belt_B_03, bone2Name=dyn_belt_B_02, swing1LimitDegrees=8, swing2LimitDegrees=10, twistLimitDegrees=0}
 {bone1Name=dyn_belt_B_04, bone2Name=dyn_belt_B_03, swing1LimitDegrees=5, swing2LimitDegrees=10, twistLimitDegrees=0}
 {bone1Name=dyn_belt_B_05, bone2Name=dyn_belt_B_04, swing1LimitDegrees=0, swing2LimitDegrees=15, twistLimitDegrees=0}
 {
  bone1Name=base_belt_F_01
  bone2Name=pelvis
  swing1LimitDegrees=45
  swing1Motion=Locked
  swing2LimitDegrees=45
  swing2Motion=Locked
  twistLimitDegrees=45
  twistMotion=Locked
 }
 {
  bone1Name=dyn_belt_F_01
  bone2Name=base_belt_F_01
  swing1LimitDegrees=5
  swing1Motion=Locked
  swing2LimitDegrees=5
  swing2Motion=Locked
  twistLimitDegrees=0
  twistMotion=Locked
 }
 {bone1Name=dyn_belt_F_02, bone2Name=dyn_belt_F_01, swing1LimitDegrees=10, swing2LimitDegrees=5, twistLimitDegrees=0}
 {bone1Name=dyn_belt_F_03, bone2Name=dyn_belt_F_02, swing1LimitDegrees=10, swing2LimitDegrees=10, twistLimitDegrees=0}
 {bone1Name=dyn_belt_F_04, bone2Name=dyn_belt_F_03, swing1LimitDegrees=10, swing2LimitDegrees=10, twistLimitDegrees=0}
 {
  bone1Name=dyn_belt_B_05
  bone2Name=dyn_belt_F_04
  swing1LimitDegrees=45
  swing1Motion=Free
  swing2LimitDegrees=45
  swing2Motion=Free
  twistLimitDegrees=45
  twistMotion=Free
 }
]
