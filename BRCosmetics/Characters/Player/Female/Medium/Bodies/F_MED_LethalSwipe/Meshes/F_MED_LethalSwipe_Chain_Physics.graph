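# Physics asset for the hanging chain and watch on the LethalSwipe medium female player body, with capsule bodies for pelvis, left thigh and left calf.
# Front and back belt anchors are angularly locked to the pelvis and feed swing-limited chain segments, with the watch branching off the first back segment.
# The last back and front segments join with free angular motion, closing the chain into a loop.
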PhysicsAsset /BRCosmetics/Characters/Player/Female/Medium/Bodies/F_MED_LethalSwipe/Meshes/F_MED_LethalSwipe_Chain_Physics
bodies {
 base_belt_B_01 [{shapeType=Capsule, length=0.505, radius=0.505, shapeName=base_belt_B_01_capsule, rotation=<rot 0 -90 0>}]
 base_belt_F_01 [{shapeType=Capsule, length=0.505, radius=0.505, shapeName=base_belt_F_01_capsule, rotation=<rot 0 -90 0>}]
 base_chain_B_01 [{shapeType=Capsule, length=0.505, radius=0.505, shapeName=base_chain_B_01_capsule, rotation=<rot 0 -90 0>}]
 base_chain_B_02 [{shapeType=Capsule, length=0.505, radius=0.505, shapeName=base_chain_B_02_capsule, rotation=<rot 0 -90 0>}]
 base_chain_F_01 [{shapeType=Capsule, length=0.505, radius=0.505, shapeName=base_chain_F_01_capsule, rotation=<rot 0 -90 0>}]
 calf_l [{shapeType=Capsule, length=28.1274, radius=12.1308, center=<-28.2813 -0.2689 -0.7494>, rotation=<rot 0 78.304 -88.5873>}]
 dyn_chain_B_01 [{shapeType=Capsule, length=0.3, radius=1.7, shapeName=dyn_chain_B_01_capsule, center=<2 -1 0>}]
 dyn_chain_B_02 [{shapeType=Capsule, length=3, radius=1, shapeName=dyn_chain_B_02_capsule, center=<2 0 0>, rotation=<rot 90 0 0>}]
 dyn_chain_B_03 [{shapeType=Capsule, length=2.5, radius=1.5, shapeName=dyn_chain_B_03_capsule, center=<2 0 0>, rotation=<rot 90 0 0>}]
 dyn_chain_B_04 [{shapeType=Capsule, length=2.5, radius=1.5, shapeName=dyn_chain_B_04_capsule, center=<2 -0.25 0>, rotation=<rot 90 0 0>}]
 dyn_chain_B_05 [{shapeType=Capsule, length=3, radius=1, shapeName=dyn_chain_B_05_capsule, center=<2 -0.15 0>, rotation=<rot 90 0 0>}]
 dyn_chain_F_01 [{shapeType=Capsule, length=3, radius=1, shapeName=dyn_chain_F_01_capsule, center=<3 0 0>, rotation=<rot 90 0 0>}]
 dyn_chain_F_02 [{shapeType=Capsule, length=3, radius=1, shapeName=dyn_chain_F_02_capsule, center=<2 -1.1352 0>, rotation=<rot 90 0 0>}]
 dyn_chain_F_03 [{shapeType=Capsule, length=3, radius=1, shapeName=dyn_chain_F_03_capsule, center=<2 0 0>, rotation=<rot 90 0 0>}]
 dyn_watch_01 [{shapeType=Capsule, length=1.5, radius=0.5, shapeName=dyn_watch_01_capsule, center=<1.6298 0.0284 0.0417>, rotation=<rot 0 91.0052 -88.8568>}]
 dyn_watch_02 [{
   shapeType=Capsule
   length=1.612
   radius=1.5989
   shapeName=dyn_watch_02_capsule
   center=<2.245 0.0023 0.3003>
   rotation=<rot 0 92.8776 -81.0437>
  }]
 pelvis [
  {shapeType=Capsule, length=11.4051, radius=9.5, center=<-1.9495 -2.3891 0>}
  {shapeType=Capsule, length=11.4051, radius=9.5, center=<-5.4711 -4.2293 0>}
 ]
 thigh_l [{shapeType=Capsule, length=21.902, radius=8.4023, center=<-15.4044 -0.4353 0.2759>, rotation=<rot 0 86.3912 -83.781>}]
}
constraints [
 {
  bone1Name=thigh_l
  bone2Name=pelvis
  swing1LimitDegrees=45
  swing1Motion=Locked
  swing2LimitDegrees=45
  swing2Motion=Locked
  twistLimitDegrees=45
  twistMotion=Locked
 }
 {bone1Name=calf_l, bone2Name=thigh_l, swing1LimitDegrees=45, swing2LimitDegrees=45, twistLimitDegrees=45}
 {bone1Name=dyn_watch_01, bone2Name=dyn_chain_B_01, swing1LimitDegrees=5, swing2LimitDegrees=20, twistLimitDegrees=0}
 {bone1Name=dyn_watch_02, bone2Name=dyn_watch_01, swing1LimitDegrees=0, swing2LimitDegrees=15, twistLimitDegrees=0}
 {
  bone1Name=base_belt_B_01
  bone2Name=pelvis
  swing1LimitDegrees=45
  swing1Motion=Locked
  swing2LimitDegrees=45
  swing2Motion=Locked
  twistLimitDegrees=45
  twistMotion=Locked
 }
 {
  bone1Name=base_belt_F_01
  bone2Name=pelvis
  swing1LimitDegrees=45
  swing1Motion=Locked
  swing2LimitDegrees=45
  swing2Motion=Locked
  twistLimitDegrees=45
  twistMotion=Locked
 }
 {
  bone1Name=base_chain_B_01
  bone2Name=base_belt_B_01
  swing1LimitDegrees=45
  swing1Motion=Locked
  swing2LimitDegrees=45
  swing2Motion=Locked
  twistLimitDegrees=45
  twistMotion=Locked
 }
 {bone1Name=dyn_chain_B_01, bone2Name=base_chain_B_01, swing1LimitDegrees=5, swing2LimitDegrees=10, twistLimitDegrees=0}
 {
  bone1Name=base_chain_B_02
  bone2Name=dyn_chain_B_01
  swing1LimitDegrees=45
  swing1Motion=Locked
  swing2LimitDegrees=45
  swing2Motion=Locked
  twistLimitDegrees=45
  twistMotion=Locked
 }
 {bone1Name=dyn_chain_B_02, bone2Name=base_chain_B_02, swing1LimitDegrees=5, swing2LimitDegrees=10, twistLimitDegrees=0}
 {bone1Name=dyn_chain_B_03, bone2Name=dyn_chain_B_02, swing1LimitDegrees=5, swing2LimitDegrees=10, twistLimitDegrees=0}
 {bone1Name=dyn_chain_B_04, bone2Name=dyn_chain_B_03, swing1LimitDegrees=5, swing2LimitDegrees=10, twistLimitDegrees=0}
 {bone1Name=dyn_chain_B_05, bone2Name=dyn_chain_B_04, swing1LimitDegrees=5, swing2LimitDegrees=10, twistLimitDegrees=0}
 {
  bone1Name=base_chain_F_01
  bone2Name=base_belt_F_01
  swing1LimitDegrees=45
  swing1Motion=Locked
  swing2LimitDegrees=45
  swing2Motion=Locked
  twistLimitDegrees=45
  twistMotion=Locked
 }
 {bone1Name=dyn_chain_F_01, bone2Name=base_chain_F_01, swing1LimitDegrees=25, swing2LimitDegrees=25, twistLimitDegrees=0}
 {bone1Name=dyn_chain_F_02, bone2Name=dyn_chain_F_01, swing1LimitDegrees=25, swing2LimitDegrees=25, twistLimitDegrees=0}
 {bone1Name=dyn_chain_F_03, bone2Name=dyn_chain_F_02, swing1LimitDegrees=5, swing2LimitDegrees=10, twistLimitDegrees=0}
 {
  bone1Name=dyn_chain_B_05
  bone2Name=dyn_chain_F_03
  swing1LimitDegrees=45
  swing1Motion=Free
  swing2LimitDegrees=45
  swing2Motion=Free
  twistLimitDegrees=45
  twistMotion=Free
 }
]
