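# Physics asset for the F_MED_LethalSwipe female medium player body cosmetic, setting up physics for its jacket around the torso.
# Capsules on spine_02 to spine_05 form a jointed torso chain.
# Seven small jacket base bodies around the body are angularly locked to spine_03.
# Each anchors a two-link dynamic capsule chain with narrow swing limits and zero-degree twist limits, tightest at the center-back root.
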PhysicsAsset /BRCosmetics/Characters/Player/Female/Medium/Bodies/F_MED_LethalSwipe/Meshes/F_MED_LethalSwipe_Physics
bodies {
 base_jacket_BL_01 [{shapeType=Capsule, length=0.505, radius=0.505, shapeName=base_jacket_BL_01_capsule, rotation=<rot 0 -90 0>}]
 base_jacket_BR_01 [{shapeType=Capsule, length=0.505, radius=0.505, shapeName=base_jacket_BR_01_capsule, rotation=<rot 0 -90 0>}]
 base_jacket_B_01 [{shapeType=Capsule, length=0.505, radius=0.505, shapeName=base_jacket_B_01_capsule, rotation=<rot 0 -90 0>}]
 base_jacket_FL_01 [{shapeType=Capsule, length=0.505, radius=0.505, shapeName=base_jacket_FL_01_capsule, rotation=<rot 0 -90 0>}]
 base_jacket_FR_01 [{shapeType=Capsule, length=0.505, radius=0.505, shapeName=base_jacket_FR_01_capsule, rotation=<rot 0 -90 0>}]
 base_jacket_L_01 [{shapeType=Capsule, length=0.505, radius=0.505, shapeName=base_jacket_L_01_capsule, rotation=<rot 0 -90 0>}]
 base_jacket_R_01 [{shapeType=Capsule, length=0.505, radius=0.505, shapeName=base_jacket_R_01_capsule, rotation=<rot 0 -90 0>}]
 dyn_jacket_BL_01 [{shapeType=Capsule, length=3, radius=2, shapeName=dyn_jacket_BL_01_capsule, center=<2 0 0>, rotation=<rot 0.0001 0.0001 0>}]
 dyn_jacket_BL_02 [{shapeType=Capsule, length=3, radius=2, shapeName=dyn_jacket_BL_02_capsule, center=<2 0 0>}]
 dyn_jacket_BR_01 [{shapeType=Capsule, length=3, radius=2, shapeName=dyn_jacket_BR_01_capsule, center=<2 0 0>}]
 dyn_jacket_BR_02 [{shapeType=Capsule, length=3, radius=2, shapeName=dyn_jacket_BR_02_capsule, center=<2 0 0>}]
 dyn_jacket_B_01 [{shapeType=Capsule, length=3, radius=2, shapeName=dyn_jacket_B_01_capsule, center=<2 0 0>}]
 dyn_jacket_B_02 [{shapeType=Capsule, length=3, radius=2, shapeName=dyn_jacket_B_02_capsule, center=<2 0 0>}]
 dyn_jacket_FL_01 [{shapeType=Capsule, length=3, radius=2, shapeName=dyn_jacket_FL_01_capsule, center=<2 0 0>}]
 dyn_jacket_FL_02 [{shapeType=Capsule, length=3, radius=2, shapeName=dyn_jacket_FL_02_capsule, center=<2 0 0>}]
 dyn_jacket_FR_01 [{shapeType=Capsule, length=3, radius=2, shapeName=dyn_jacket_FR_01_capsule, center=<2 0 0>}]
 dyn_jacket_FR_02 [{shapeType=Capsule, length=3, radius=2, shapeName=dyn_jacket_FR_02_capsule, center=<2 0 0>}]
 dyn_jacket_L_01 [{shapeType=Capsule, length=3, radius=2, shapeName=dyn_jacket_L_01_capsule, center=<2 0 0>}]
 dyn_jacket_L_02 [{shapeType=Capsule, length=3, radius=2, shapeName=dyn_jacket_L_02_capsule, center=<2 0 0>}]
 dyn_jacket_R_01 [{shapeType=Capsule, length=3, radius=2, shapeName=dyn_jacket_R_01_capsule, center=<2 0 0>}]
 dyn_jacket_R_02 [{shapeType=Capsule, length=3, radius=2, shapeName=dyn_jacket_R_02_capsule, center=<2 0 0>}]
 spine_02 [{
   shapeType=Capsule
   length=9.6032
   radius=8.3603
   shapeName=spine_02_capsule
   center=<2.9545 1.3153 0.0466>
   rotation=<rot -8.7964 -90 2.7702>
  }]
 spine_03 [{shapeType=Capsule, length=3.155, radius=8.5419, center=<0 2.2571 0>}]
 spine_04 [{shapeType=Capsule, length=9.3392, radius=8, center=<0 2.6973 0>}]
 spine_05 [{shapeType=Capsule, length=21.408, radius=7.1636}]
}
constraints [
 {bone1Name=spine_03, bone2Name=spine_02, swing1LimitDegrees=45, swing2LimitDegrees=45, twistLimitDegrees=45}
 {bone1Name=spine_04, bone2Name=spine_03, swing1LimitDegrees=45, swing2LimitDegrees=45, twistLimitDegrees=45}
 {bone1Name=spine_05, bone2Name=spine_04, swing1LimitDegrees=45, swing2LimitDegrees=45, twistLimitDegrees=45}
 {
  bone1Name=base_jacket_BL_01
  bone2Name=spine_03
  swing1LimitDegrees=45
  swing1Motion=Locked
  swing2LimitDegrees=45
  swing2Motion=Locked
  twistLimitDegrees=45
  twistMotion=Locked
 }
 {bone1Name=dyn_jacket_BL_01, bone2Name=base_jacket_BL_01, swing1LimitDegrees=10, swing2LimitDegrees=10, twistLimitDegrees=0}
 {bone1Name=dyn_jacket_BL_02, bone2Name=dyn_jacket_BL_01, swing1LimitDegrees=10, swing2LimitDegrees=10, twistLimitDegrees=0}
 {
  bone1Name=base_jacket_BR_01
  bone2Name=spine_03
  swing1LimitDegrees=45
  swing1Motion=Locked
  swing2LimitDegrees=45
  swing2Motion=Locked
  twistLimitDegrees=45
  twistMotion=Locked
 }
 {bone1Name=dyn_jacket_BR_01, bone2Name=base_jacket_BR_01, swing1LimitDegrees=10, swing2LimitDegrees=10, twistLimitDegrees=0}
 {bone1Name=dyn_jacket_BR_02, bone2Name=dyn_jacket_BR_01, swing1LimitDegrees=10, swing2LimitDegrees=10, twistLimitDegrees=0}
 {
  bone1Name=base_jacket_B_01
  bone2Name=spine_03
  swing1LimitDegrees=45
  swing1Motion=Locked
  swing2LimitDegrees=45
  swing2Motion=Locked
  twistLimitDegrees=45
  twistMotion=Locked
 }
 {bone1Name=dyn_jacket_B_01, bone2Name=base_jacket_B_01, swing1LimitDegrees=5, swing2LimitDegrees=5, twistLimitDegrees=0}
 {bone1Name=dyn_jacket_B_02, bone2Name=dyn_jacket_B_01, swing1LimitDegrees=10, swing2LimitDegrees=10, twistLimitDegrees=0}
 {
  bone1Name=base_jacket_FL_01
  bone2Name=spine_03
  swing1LimitDegrees=45
  swing1Motion=Locked
  swing2LimitDegrees=45
  swing2Motion=Locked
  twistLimitDegrees=45
  twistMotion=Locked
 }
 {bone1Name=dyn_jacket_FL_01, bone2Name=base_jacket_FL_01, swing1LimitDegrees=10, swing2LimitDegrees=10, twistLimitDegrees=0}
 {bone1Name=dyn_jacket_FL_02, bone2Name=dyn_jacket_FL_01, swing1LimitDegrees=10, swing2LimitDegrees=10, twistLimitDegrees=0}
 {
  bone1Name=base_jacket_FR_01
  bone2Name=spine_03
  swing1LimitDegrees=45
  swing1Motion=Locked
  swing2LimitDegrees=45
  swing2Motion=Locked
  twistLimitDegrees=45
  twistMotion=Locked
 }
 {bone1Name=dyn_jacket_FR_01, bone2Name=base_jacket_FR_01, swing1LimitDegrees=10, swing2LimitDegrees=10, twistLimitDegrees=0}
 {bone1Name=dyn_jacket_FR_02, bone2Name=dyn_jacket_FR_01, swing1LimitDegrees=10, swing2LimitDegrees=10, twistLimitDegrees=0}
 {
  bone1Name=base_jacket_L_01
  bone2Name=spine_03
  swing1LimitDegrees=45
  swing1Motion=Locked
  swing2LimitDegrees=45
  swing2Motion=Locked
  twistLimitDegrees=45
  twistMotion=Locked
 }
 {bone1Name=dyn_jacket_L_01, bone2Name=base_jacket_L_01, swing1LimitDegrees=10, swing2LimitDegrees=10, twistLimitDegrees=0}
 {bone1Name=dyn_jacket_L_02, bone2Name=dyn_jacket_L_01, swing1LimitDegrees=10, swing2LimitDegrees=10, twistLimitDegrees=0}
 {
  bone1Name=base_jacket_R_01
  bone2Name=spine_03
  swing1LimitDegrees=45
  swing1Motion=Locked
  swing2LimitDegrees=45
  swing2Motion=Locked
  twistLimitDegrees=45
  twistMotion=Locked
 }
 {bone1Name=dyn_jacket_R_01, bone2Name=base_jacket_R_01, swing1LimitDegrees=10, swing2LimitDegrees=10, twistLimitDegrees=0}
 {bone1Name=dyn_jacket_R_02, bone2Name=dyn_jacket_R_01, swing1LimitDegrees=10, swing2LimitDegrees=10, twistLimitDegrees=0}
]
