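# Character-part AnimBlueprint for the F_MED_LethalSwipe body, adding corrective rigging and secondary motion to the pose copied from its parent mesh.
# A Control Rig takes crouch, kneeslide, grind-rail and DBNO alphas, then attach_backpack gets a small additive offset.
# Three chained RigidBody nodes simulate belt, chain and main physics assets in bone space, followed by AnimDynamics on a bracelet.
# The EventGraph sets state alphas and computes gravity and rigid-body settings from a parameter asset.
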AnimBlueprint /BRCosmetics/Characters/Player/Female/Medium/Bodies/F_MED_LethalSwipe/Meshes/F_MED_LethalSwipe_AnimBP
target_skeleton=&/F_MED_LethalSwipe_Skeleton
graphs [
 AnimGraph {nodes [
   Output Pose(Result=Use cached pose 'Dynamic'()) {layerGroup=None}
   n1 = Control Rig!(
    Source=Copy Pose From Mesh() {bUseAttachedParent=true}
    Alpha
    crouching_alpha:double=Get crouching_alpha() -> (crouching_alpha:double)
    Has_Rban:bool=Get bRigidBodyEnabled()
    Kneeslide:double=Get kneesliding_alpha() -> (kneesliding_alpha:double)
    grindrail_alpha:double=Get grindrail_alpha() -> (grindrail_alpha:double)
    DBNO_alpha:double=Get DBNO_alpha() -> (DBNO_alpha:double)
   ) -> (Pose) {controlRigAssetReference {blueprintRigClass=&/F_MED_LethalSwipe_CtrlRig.F_MED_LethalSwipe_CtrlRig_C}}
   Save cached pose 'ControlRig'(Pose=@n1) {cacheName=ControlRig}
   n2 = RigidBody!(ComponentPose=Local To Component(LocalPose=Use cached pose 'Transforms'()), Alpha, OverrideWorldGravity:Vector="0,0,0", ExternalForce) -> (Pose) {
    bOverrideWorldGravity=true
    overridePhysicsAsset=&/F_MED_LethalSwipe_Belt_Physics
    simulationSpace=BaseBoneSpace
    baseBoneRef {boneName=pelvis}
    simSpaceSettings {
     maxAngularAcceleration=40
     maxAngularVelocity=8
     maxLinearAcceleration=3500
     maxLinearVelocity=700
     velocityScaleZ=0.4
     worldAlpha=1
     externalLinearDragV=<0.3 0.2 0.2>
     externalLinearVelocity=<0.3 0.2 0.2>
    }
   }
   n3 = RigidBody(ComponentPose=@n2, OverrideWorldGravity="0,0,0") {
    bOverrideWorldGravity=true
    overridePhysicsAsset=&/F_MED_LethalSwipe_Chain_Physics
    simulationSpace=BaseBoneSpace
    baseBoneRef {boneName=pelvis}
    simSpaceSettings {
     maxAngularAcceleration=40
     maxAngularVelocity=8
     maxLinearAcceleration=4000
     maxLinearVelocity=800
     velocityScaleZ=0.5
     worldAlpha=0.8
     externalLinearDragV=<0.3 0.2 0.2>
     externalLinearVelocity=<0.3 0.2 0.2>
    }
   }
   n4 = RigidBody!(
    ComponentPose=@n3
    Alpha
    OverrideWorldGravity="0,0,0"
    ExternalForce
    SimSpaceSettings:Sim Space Settings Structure=(WorldAlpha=0.2,VelocityScaleZ=1,DampingAlpha=1,MaxLinearVelocity=800,MaxAngularVelocity=8,MaxLinearAcceleration=4000,MaxAngularAcceleration=40,ExternalLinearDragV=(X=0.3,Y=0.2,Z=0.2),ExternalLinearVelocity=(X=0,Y=0,Z=0),ExternalAngularVelocity=(X=0,Y=0,Z=0))
   ) -> (Pose) {
    bOverrideWorldGravity=true
    overridePhysicsAsset=&/F_MED_LethalSwipe_Physics
    simulationSpace=BaseBoneSpace
    baseBoneRef {boneName=spine_05}
    simSpaceSettings {
     maxAngularAcceleration=40
     maxAngularVelocity=8
     maxLinearAcceleration=4000
     maxLinearVelocity=800
     worldAlpha=0.2
     externalLinearDragV=<0.3 0.2 0.2>
    }
   }
   n5 = Anim Dynamics - Bone: dyn_bracelet_01(ComponentPose=@n4, Alpha=1, ExternalForce="0,0,0") {
    simulationSpace=BoneRelative
    boundBone {boneName=dyn_bracelet_01}
    physicsBodyDefinitions [{
      boundBone {boneName=dyn_bracelet_01}
      boxExtents=<20 20 20>
      constraintSetup {angularLimitsMax=<0 0 0>, angularLimitsMin=<0 0 -15>, angularTarget=<0 0 0>}
      localJointOffset=<2 0 0>
     }]
    relativeSpaceBone {boneName=root}
   }
   Save cached pose 'Dynamic'(Pose=Component To Local(ComponentPose=@n5)) {cacheName=Dynamic}
   n6 = "Transform (Modify) Bone - Bone: attach_backpack"(ComponentPose=Local To Component(LocalPose=Use cached pose 'ControlRig'()), Alpha=1, Translation="0,4,0", Rotation="0,0,0", Scale="1,1,1") {translationMode=BMM_Additive, translationSpace=BCS_BoneSpace, boneToModify {boneName=attach_backpack}, translation {y=4}}
   Save cached pose 'Transforms'(Pose=Component To Local(ComponentPose=@n6)) {cacheName=Transforms}
  ]}
 EventGraph {nodes [
   Event BlueprintUpdateAnimation()
   > n1 = Sequence()
   > n2 = "Gravity Override State Function (Parameter Asset) (Nativized)"(
    CharacterPartAnimInstance=Self-Reference()
    GameTimeInSeconds=GetGameTimeInSeconds()
    Parameters=&/F_MED_LethalSwipe_Parameters
   )
   > Set GravityOverride(GravityOverride=@n2)
   > Set RBAN Settings Out(RBAN Settings Out=@n2.OutRigidBodyAnimNodeParameters)
   n3 = Branch(execute=@n1.then_1, Condition=Get bIsDBNO())
   > Set DBNO_alpha(DBNO_alpha:double=1) -> (Output_Get:double)
   Set DBNO_alpha(execute=@n3.else, DBNO_alpha=0)
   Event OnBeginCrouching()
   > Set crouching_alpha(crouching_alpha:double=1) -> (Output_Get:double)
   Event OnEndCrouching()
   > Set crouching_alpha(crouching_alpha=0)
   Event OnBeginKneeSliding() {is=title K2Node_Event}
   > Set kneesliding_alpha(kneesliding_alpha:double=1) -> (Output_Get:double)
   Event OnEndKneeSliding() {is=title K2Node_Event}
   > Set kneesliding_alpha(kneesliding_alpha=0)
   Event OnEnterGrindRail() {is=title K2Node_Event}
   > Set grindrail_alpha(grindrail_alpha:double=1) -> (Output_Get:double)
   Event OnExitGrindRail() {is=title K2Node_Event}
   > Set grindrail_alpha(grindrail_alpha=0)
  ]}
]
variables [
 crouching_alpha {type=number}
 DBNO_alpha {type=number}
 GravityOverride {type=Vector}
 grindrail_alpha {type=number}
 kneesliding_alpha {type=number}
 RBAN Settings Out {type=FortRigidBodyAnimNodeParameters}
]
