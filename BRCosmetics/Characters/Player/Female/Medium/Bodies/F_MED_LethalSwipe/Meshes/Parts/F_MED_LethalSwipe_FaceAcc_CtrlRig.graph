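# Control Rig for the LethalSwipe female character's face accessory, applying corrective offsets to hair bones as the upper body moves.
# Relative rotations, remaps, distances and spherical pose readers on head, spine, clavicles and right upper arm weight Modify Transforms units.
# The live sequence blends aim corrections through interpolated alphas and adds further fixes when hasRBAN is set.
# A second sequence, holding crouch and other aim corrections, is disconnected from execution.
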ControlRigBlueprint /BRCosmetics/Characters/Player/Female/Medium/Bodies/F_MED_LethalSwipe/Meshes/Parts/F_MED_LethalSwipe_FaceAcc_CtrlRig
graphs [
 Multiply_2_1_ContainedGraph {nodes [
   n1 = RigVMFunctionEntryNode()
   RigVMFunctionReturnNode(Result=&RigVMFunction_MathDoubleMul(B=@n1.C, A=&RigVMFunction_MathDoubleMul(A=@n1.A, B=@n1.B).Result).Result)
  ]}
 Multiply_2_1_ContainedGraph {nodes [
   n1 = RigVMFunctionEntryNode()
   RigVMFunctionReturnNode(Result=&RigVMFunction_MathDoubleMul(B=@n1.C, A=&RigVMFunction_MathDoubleMul(A=@n1.A, B=@n1.B).Result).Result)
  ]}
 Multiply_2_1_ContainedGraph {nodes [
   n1 = RigVMFunctionEntryNode()
   RigVMFunctionReturnNode(Result=&RigVMFunction_MathDoubleMul(B=@n1.C, A=&RigVMFunction_MathDoubleMul(A=@n1.A, B=@n1.B).Result).Result)
  ]}
 New Function_ContainedGraph {nodes [
   RigVMFunctionEntryNode()
   RigUnit_GetRelativeTransformForItem = &RigUnit_GetRelativeTransformForItem(Child=(Type=Bone,Name="head"), Parent=(Type=Bone,Name="spine_05"), bChildInitial=False, bParentInitial=False)
   RigVMFunction_MathQuaternionToEuler = &RigVMFunction_MathQuaternionToEuler(RotationOrder=ZYX, Value=@RigUnit_GetRelativeTransformForItem.RelativeTransform.Rotation)
   > RigVMFunctionReturnNode(Result=@RigVMFunction_MathQuaternionToEuler.Result)
  ]}
 New Function_ContainedGraph {nodes [
   RigVMFunctionEntryNode()
   RigUnit_GetRelativeTransformForItem = &RigUnit_GetRelativeTransformForItem(Child=(Type=Bone,Name="head"), Parent=(Type=Bone,Name="spine_05"), bChildInitial=False, bParentInitial=False)
   RigVMFunction_MathQuaternionToEuler = &RigVMFunction_MathQuaternionToEuler(RotationOrder=ZYX, Value=@RigUnit_GetRelativeTransformForItem.RelativeTransform.Rotation)
   > RigVMFunctionReturnNode(Result=@RigVMFunction_MathQuaternionToEuler.Result)
  ]}
 RigVMFunctionLibrary {nodes [
   F_Head Rotation = RigVMCollapseNode()
   F_Head Rotation_1 = RigVMCollapseNode()
  ]}
 RigVMFunction_MathBoolAnd_1_ContainedGraph {nodes [
   n1 = RigVMFunctionEntryNode()
   RigVMFunction_MathBoolAnd = &RigVMFunction_MathBoolAnd(A=@n1.A, B=@n1.B)
   RigVMFunctionReturnNode(Result=&RigVMFunction_MathBoolAnd(B=@n1.C, A=@RigVMFunction_MathBoolAnd.Result).Result)
  ]}
 RigVMFunction_Sequence_1_ContainedGraph {nodes [
   n1 = RigVMFunctionEntryNode()
   RigVMFunction_Sequence = &RigVMFunction_Sequence(ExecuteContext=@n1.ExecuteContext)
   n2 = &RigVMFunction_Sequence(ExecuteContext=@RigVMFunction_Sequence.B)
   n3 = &RigVMFunction_Sequence(ExecuteContext=@n2.B)
   RigVMFunctionReturnNode(A=@RigVMFunction_Sequence.A, B=@n2.A, C=@n3.A, D=@n3.B)
  ]}
 RigVMFunction_Sequence_1_ContainedGraph {nodes [
   n1 = RigVMFunctionEntryNode()
   RigVMFunction_Sequence = &RigVMFunction_Sequence(ExecuteContext=@n1.ExecuteContext)
   n2 = &RigVMFunction_Sequence(ExecuteContext=@RigVMFunction_Sequence.B)
   n3 = &RigVMFunction_Sequence(ExecuteContext=@n2.B)
   n4 = &RigVMFunction_Sequence(ExecuteContext=@n3.B)
   RigVMFunctionReturnNode(A=@RigVMFunction_Sequence.A, B=@n2.A, C=@n3.A, D=@n4.A, E=@n4.B)
  ]}
 RigVMModel {nodes [
   RigUnit_BeginExecution = &RigUnit_BeginExecution()
   RigVMFunction_Sequence_1 = RigVMAggregateNode(ExecuteContext=@RigUnit_BeginExecution.ExecutePin)
   n1 = &RigVMFunction_ControlFlowBranch(Condition=RigVMVariableNode() {variable=hasRBAN}.Value, ExecuteContext=@RigVMFunction_Sequence_1.D)
   n2 = &RigUnit_SphericalPoseReader(
    ActiveRegionScaleFactors=(PositiveWidth=1,NegativeWidth=1,PositiveHeight=1,NegativeHeight=1)
    ActiveRegionSize=0.01
    Debug=(bDrawDebug=false,bDraw2D=False,bDrawLocalAxes=False,DebugScale=25,DebugSegments=20,DebugThickness=0.25)
    DriverAxis=(X=0,Y=1,Z=0)
    DriverItem=(Type=Bone,Name="head")
    FalloffRegionScaleFactors=(PositiveWidth=1,NegativeWidth=1,PositiveHeight=1,NegativeHeight=1)
    FalloffSize=0.1
    FlipHeightScaling=False
    FlipWidthScaling=False
    OptionalParentItem=(Type=Bone,Name="clavicle_r")
    RotationOffset=(X=0,Y=-87,Z=-5)
    ExecutePin=@n1.True
   )
   > &RigUnit_ModifyTransforms(
    ItemToModify=((Item=(Type=Bone,Name="base_hair_R_01"),Transform=(Rotation=(X=0,Y=0,Z=0,W=1),Translation=(X=8,Y=0,Z=0),Scale3D=(X=1,Y=1,Z=1))),(Item=(Type=Bone,Name="dyn_hair_R_02"),Transform=(Rotation=(X=0,Y=0,Z=0,W=1),Translation=(X=0,Y=0,Z=0),Scale3D=(X=1,Y=1,Z=1))))
    WeightMinimum=0
    Weight=@n2.OutputParam
   )
   n3 = &RigVMFunction_MathVectorDistance(
    A=&RigVMFunction_MathQuaternionToEuler(
     RotationOrder=YZX
     Value=&RigUnit_GetRelativeTransformForItem(Child=(Type=Bone,Name="clavicle_r"), Parent=(Type=Bone,Name="head"), bChildInitial=False, bParentInitial=False).RelativeTransform.Rotation
    ).Result
    B=/Script/RigVM.RigVMFunction_MathVectorMake(X=-127.1, Y=-156, Z=-61.94) {is=struct RigVMUnitNode}.Result
   )
   > &RigUnit_ModifyTransforms(
    ItemToModify=((Item=(Type=Bone,Name="base_hair_lf"),Transform=(Rotation=(X=0,Y=0,Z=0,W=1),Translation=(X=0,Y=0,Z=0),Scale3D=(X=1,Y=1,Z=1))),(Item=(Type=Bone,Name="base_hair_cr"),Transform=(Rotation=(X=0,Y=0,Z=0,W=1),Translation=(X=0,Y=0,Z=0),Scale3D=(X=1,Y=1,Z=1))))
    WeightMinimum=0
    Weight=&RigVMFunction_KalmanFloat(
     BufferSize=2
     Value=&RigVMFunction_MathDoubleRemap(SourceMaximum=3, SourceMinimum=20, TargetMaximum=1, TargetMinimum=0, Value=@n3.Result).Result
    ) {template="KalmanFilter::Execute(in Value,in BufferSize,out Result)"}.Result
   )
   F_Head Rotation = RigVMFunctionReferenceNode()
   Remap = &RigVMFunction_MathDoubleRemap(SourceMaximum=-80, SourceMinimum=0, TargetMaximum=1, TargetMinimum=0, Value=@"F_Head Rotation".Result.Z)
   &RigUnit_ModifyTransforms(
    ItemToModify="((Item=(Type=Bone,Name=\"base_hair_C_01\"),Transform=(Rotation=(X=0,Y=0,Z=0,W=-1),Translation=(X=-13.88,Y=-1.9678,Z=-0.3641),Scale3D=(X=1,Y=1,Z=1))),(Item=(Type=Bone,Name=\"base_hair_L_01\"),Transform=(Rotation=(X=0,Y=0,Z=0,W=1),Translation=(X=0,Y=0,Z=0),Scale ...(truncated)"
    WeightMinimum=0
    Weight=@Remap.Result
    ExecutePin=@RigVMFunction_Sequence_1.A
   )
   > &RigUnit_SphericalPoseReader(
    ActiveRegionScaleFactors=(PositiveWidth=1,NegativeWidth=1,PositiveHeight=1,NegativeHeight=1)
    ActiveRegionSize=0.07
    Debug=(bDrawDebug=false,bDraw2D=False,bDrawLocalAxes=False,DebugScale=25,DebugSegments=20,DebugThickness=0.25)
    DriverAxis=(X=1,Y=0,Z=0)
    DriverItem=(Type=Bone,Name="clavicle_l")
    FalloffRegionScaleFactors=(PositiveWidth=1,NegativeWidth=1,PositiveHeight=1,NegativeHeight=1)
    FalloffSize=0.2
    FlipHeightScaling=False
    FlipWidthScaling=False
    OptionalParentItem=(Type=Bone,Name="None")
    RotationOffset=(X=0,Y=-40,Z=0)
   )
   RigVMFunction_MathBoolAnd = RigVMAggregateNode(
    C=true
    B=&RigVMFunction_MathBoolNot(Value=RigVMVariableNode() {variable=hasRBAN}.Value).Result
    A=RigVMVariableNode() {variable=isaiming}.Value
   )
   RigVMFunction_MathBoolToFloat = &RigVMFunction_MathBoolToFloat(Value=@RigVMFunction_MathBoolAnd.Result)
   n4 = &RigVMFunction_AlphaInterp(InterpSpeedDecreasing=8, InterpSpeedIncreasing=10, Scale=1, bClampResult=true, Value=@RigVMFunction_MathBoolToFloat.Result)
   > &RigUnit_ModifyTransforms(
    ItemToModify="((Item=(Type=Bone,Name=\"dyn_hair_R_02\"),Transform=(Rotation=(X=0.1188,Y=0.1825,Z=0.1106,W=0.9697),Translation=(X=0,Y=0,Z=0),Scale3D=(X=1,Y=1,Z=1))),(Item=(Type=Bone,Name=\"dyn_hair_C_02\"),Transform=(Rotation=(X=0,Y=0,Z=0.0436,W=0.999),Translation=(X=0,Y=0,Z ...(truncated)"
    WeightMinimum=0
    Weight=@n4.Result
   )
   n5 = &RigVMFunction_MathQuaternionToEuler(
    RotationOrder=XZY
    Value=&RigUnit_GetRelativeTransformForItem(Child=(Type=Bone,Name="head"), Parent=(Type=Bone,Name="root"), bChildInitial=False, bParentInitial=False).RelativeTransform.Rotation
   )
   n6 = &RigVMFunction_MathVectorDistance(
    A=&RigVMFunction_MathQuaternionToEuler(
     RotationOrder=XZY
     Value=&RigUnit_GetRelativeTransformForItem(Child=(Type=Bone,Name="head"), Parent=(Type=Bone,Name="clavicle_r"), bChildInitial=False, bParentInitial=False).RelativeTransform.Rotation
    ).Result
    B=&RigVMFunction_MathVectorMake(X=130.54, Y=161.807, Z=18.695).Result
   )
   &RigUnit_ModifyTransforms(
    ItemToModify="((Item=(Type=Bone,Name=\"base_hair_C_01\"),Transform=(Rotation=(X=0,Y=-0.0872,Z=0,W=0.9962),Translation=(X=-3.2325,Y=0,Z=0),Scale3D=(X=1,Y=1,Z=1))),(Item=(Type=Bone,Name=\"base_hair_R_01\"),Transform=(Rotation=(X=0,Y=0,Z=0,W=1),Translation=(X=0,Y=0,Z=0),Scale3 ...(truncated)"
    WeightMinimum=0
    Weight=&RigVMFunction_MathDoubleMul(
     B=&RigVMFunction_MathDoubleRemap(SourceMaximum=-55, SourceMinimum=-10, TargetMaximum=1, TargetMinimum=0, Value=@n5.Result.Z).Result
     A=&RigVMFunction_MathDoubleRemap(SourceMaximum=5, SourceMinimum=40, TargetMaximum=1, TargetMinimum=0, Value=@n6.Result).Result
    ).Result
    ExecutePin=@RigVMFunction_Sequence_1.C
   )
   F_Head Rotation_1 = RigVMFunctionReferenceNode()
   Power = /Script/RigVM.RigVMFunction_MathDoublePow(
    B=0.7
    A=&RigVMFunction_MathDoubleRemap(SourceMaximum=-55, SourceMinimum=0, TargetMaximum=1, TargetMinimum=0, Value=@"F_Head Rotation_1".Result.Z).Result
   ) {is=struct RigVMUnitNode, template="Power::Execute(in A,in B,out Result)"}
   &RigUnit_ModifyTransforms(
    ItemToModify="((Item=(Type=Bone,Name=\"base_hair_C_01\"),Transform=(Rotation=(X=0,Y=0,Z=0,W=1),Translation=(X=3.6917,Y=1.7857,Z=0),Scale3D=(X=1,Y=1,Z=1))),(Item=(Type=Bone,Name=\"base_hair_L_01\"),Transform=(Rotation=(X=0,Y=0,Z=0,W=1),Translation=(X=0,Y=0,Z=0),Scale3D=(X=1, ...(truncated)"
    WeightMinimum=0
    Weight=@Power.Result
    ExecutePin=@RigVMFunction_Sequence_1.B
   )
   n7 = &RigVMFunction_MathQuaternionToEuler(
    RotationOrder=ZYX
    Value=&RigUnit_GetRelativeTransformForItem(Child=(Type=Bone,Name="head"), Parent=(Type=Null,Name="head_null"), bChildInitial=False, bParentInitial=False).RelativeTransform.Rotation
   )
   n8 = &RigUnit_GetRelativeTransformForItem(Child=(Type=Bone,Name="clavicle_r"), Parent=(Type=Null,Name="clavicle_r_null"), bChildInitial=False, bParentInitial=False)
   Multiply_4 = RigVMAggregateNode(
    B=&RigVMFunction_MathDoubleRemap(SourceMaximum=-20, SourceMinimum=5, TargetMaximum=1, TargetMinimum=0, Value=@n7.Result.Z).Result
    A=&RigVMFunction_MathDoubleRemap(SourceMaximum=-29, SourceMinimum=-5, TargetMaximum=1, TargetMinimum=0, Value=@n7.Result.Y).Result
    C=&RigVMFunction_MathDoubleRemap(
     SourceMaximum=-14
     SourceMinimum=0
     TargetMaximum=1
     TargetMinimum=0
     Value=&RigVMFunction_MathQuaternionToEuler(RotationOrder=ZYX, Value=@n8.RelativeTransform.Rotation).Result.Y
    ).Result
   )
   > &RigUnit_ModifyTransforms(
    ItemToModify="((Item=(Type=Bone,Name=\"base_hair_C_01\"),Transform=(Rotation=(X=0,Y=0,Z=0,W=1),Translation=(X=0,Y=0,Z=0),Scale3D=(X=1,Y=1,Z=1))),(Item=(Type=Bone,Name=\"base_hair_L_01\"),Transform=(Rotation=(X=0,Y=0,Z=0,W=1),Translation=(X=0,Y=0,Z=0),Scale3D=(X=1,Y=1,Z=1))) ...(truncated)"
    WeightMinimum=0
    Weight=@Multiply_4.Result
   )
   > n9 = &RigUnit_SphericalPoseReader(
    ActiveRegionScaleFactors=(PositiveWidth=1,NegativeWidth=0.5,PositiveHeight=0.75,NegativeHeight=0.75)
    ActiveRegionSize=0.1
    Debug=(bDrawDebug=false,bDraw2D=False,bDrawLocalAxes=False,DebugScale=25,DebugSegments=20,DebugThickness=0.25)
    DriverAxis=(X=0,Y=1,Z=0)
    DriverItem=(Type=Bone,Name="head")
    FalloffRegionScaleFactors=(PositiveWidth=1,NegativeWidth=0.5,PositiveHeight=1,NegativeHeight=1)
    FalloffSize=0.2
    FlipHeightScaling=False
    FlipWidthScaling=False
    OptionalParentItem=(Type=Bone,Name="None")
    RotationOffset=(X=90,Y=0,Z=-30)
   )
   > n10 = &RigUnit_SphericalPoseReader(
    ActiveRegionScaleFactors=(PositiveWidth=1,NegativeWidth=1,PositiveHeight=1,NegativeHeight=1)
    ActiveRegionSize=0.025
    Debug=(bDrawDebug=false,bDraw2D=False,bDrawLocalAxes=False,DebugScale=25,DebugSegments=20,DebugThickness=0.25)
    DriverAxis=(X=-1,Y=0,Z=0)
    DriverItem=(Type=Bone,Name="upperarm_r")
    FalloffRegionScaleFactors=(PositiveWidth=1,NegativeWidth=1,PositiveHeight=1,NegativeHeight=1)
    FalloffSize=0.2
    FlipHeightScaling=False
    FlipWidthScaling=False
    OptionalParentItem=(Type=Bone,Name="None")
    RotationOffset=(X=-115,Y=0,Z=25)
   )
   n11 = &RigVMFunction_MathDoubleMul(A=@n9.OutputParam, B=@n10.OutputParam)
   > &RigUnit_ModifyTransforms(
    ItemToModify=((Item=(Type=Bone,Name="base_hair_R_01"),Transform=(Rotation=(X=-0.2981,Y=-0.0393,Z=-0.1245,W=0.9456),Translation=(X=2.1384,Y=-0.7762,Z=-1.7946),Scale3D=(X=1,Y=1,Z=1))))
    WeightMinimum=0
    Weight=@n11.Result
   )
   n12 = &RigVMFunction_AlphaInterp(InterpSpeedDecreasing=10, InterpSpeedIncreasing=10, Scale=1, bClampResult=true, Value=RigVMVariableNode() {variable=aimalpha}.Value)
   n13 = &RigVMFunction_MathDoubleMul(B=@n12.Result, A=@n11.Result)
   > &RigUnit_ModifyTransforms(
    ItemToModify="((Item=(Type=Bone,Name=\"base_hair_R_01\"),Transform=(Rotation=(X=0.2147,Y=0.1888,Z=0.3353,W=0.8977),Translation=(X=-1.8439,Y=1.4173,Z=3.0071),Scale3D=(X=1,Y=1,Z=1))),(Item=(Type=Bone,Name=\"dyn_hair_R_02\"),Transform=(Rotation=(X=0,Y=-0.3007,Z=0,W=0.9537),Tra ...(truncated)"
    WeightMinimum=0
    Weight=@n13.Result
   )
   RigVMFunction_Sequence = RigVMAggregateNode(ExecuteContext=())
   n14 = &RigVMFunction_MathQuaternionToEuler(
    RotationOrder=XZY
    Value=&RigUnit_GetRelativeTransformForItem(Child=(Type=Bone,Name="head"), Parent=(Type=Bone,Name="spine_01"), bChildInitial=False, bParentInitial=False).RelativeTransform.Rotation
   )
   Multiply = &RigVMFunction_MathDoubleMul(
    B=1
    A=&RigVMFunction_MathDoubleRemap(SourceMaximum=30, SourceMinimum=6, TargetMaximum=1, TargetMinimum=0, Value=@n14.Result.X).Result
   )
   &RigUnit_ModifyTransforms(
    ItemToModify=((Item=(Type=Bone,Name="base_hair_R_01"),Transform=(Rotation=(X=0,Y=0,Z=0.1736,W=0.9848),Translation=(X=-0.31,Y=-0.8519,Z=0),Scale3D=(X=1,Y=1,Z=1))))
    WeightMinimum=0
    Weight=@Multiply.Result
    ExecutePin=@RigVMFunction_Sequence.C
   )
   n15 = &RigVMFunction_MathDoubleMul(
    B=1
    A=&RigVMFunction_MathDoubleRemap(SourceMaximum=-6, SourceMinimum=-30, TargetMaximum=0, TargetMinimum=1, Value=@n14.Result.X).Result
   )
   > &RigUnit_ModifyTransforms(
    ItemToModify=((Item=(Type=Bone,Name="base_hair_BR_00"),Transform=(Rotation=(X=0.0858,Y=0.0151,Z=0,W=0.9962),Translation=(X=-0.3101,Y=-0.8519,Z=0),Scale3D=(X=1,Y=1,Z=1))))
    WeightMinimum=0
    Weight=@n15.Result
   )
   RigVMFunction_ControlFlowBranch = &RigVMFunction_ControlFlowBranch(Condition=RigVMVariableNode() {variable=iscrouching}.Value, ExecuteContext=@RigVMFunction_Sequence.A)
   F_Head Rotation_4 = RigVMFunctionReferenceNode()
   n16 = &RigVMFunction_MathDoubleRemap(SourceMaximum=-55, SourceMinimum=0, TargetMaximum=1, TargetMinimum=0, Value=@"F_Head Rotation_4".Result.Z)
   &RigUnit_ModifyTransforms(
    ItemToModify="((Item=(Type=Bone,Name=\"base_hair_C_01\"),Transform=(Rotation=(X=0,Y=0,Z=-0.5736,W=0.8192),Translation=(X=-1.5635,Y=1.8633,Z=0),Scale3D=(X=1,Y=1,Z=1))),(Item=(Type=Bone,Name=\"base_hair_L_01\"),Transform=(Rotation=(X=0,Y=0,Z=0,W=1),Translation=(X=0,Y=0,Z=0),S ...(truncated)"
    WeightMinimum=0
    Weight=@n16.Result
    ExecutePin=@RigVMFunction_ControlFlowBranch.False
   )
   > RigVMVariableNode() {variable=aimalpha}
   n17 = &RigUnit_GetRelativeTransformForItem(Child=(Type=Bone,Name="head"), Parent=(Type=Null,Name="head_null"), bChildInitial=False, bParentInitial=False)
   n18 = &RigUnit_GetRelativeTransformForItem(Child=(Type=Bone,Name="clavicle_r"), Parent=(Type=Null,Name="clavicle_r_null"), bChildInitial=False, bParentInitial=False)
   Multiply_2 = RigVMAggregateNode(
    A=1
    B=&RigVMFunction_MathDoubleRemap(
     SourceMaximum=-20
     SourceMinimum=5
     TargetMaximum=1
     TargetMinimum=0
     Value=&RigVMFunction_MathQuaternionToEuler(RotationOrder=ZYX, Value=@n17.RelativeTransform.Rotation).Result.Z
    ).Result
    C=&RigVMFunction_MathDoubleRemap(
     SourceMaximum=-14
     SourceMinimum=0
     TargetMaximum=1
     TargetMinimum=0
     bClamp=False
     Value=&RigVMFunction_MathQuaternionToEuler(RotationOrder=ZYX, Value=@n18.RelativeTransform.Rotation).Result.Y
    ).Result
   )
   > RigVMVariableNode(Value=@Multiply_2.Result) {variable=headrotationZ}
   AlphaInterp = &RigVMFunction_AlphaInterp(
    InterpSpeedDecreasing=5
    InterpSpeedIncreasing=10
    Scale=1
    bClampResult=true
    bMapRange=false
    Value=RigVMDispatchNode(False=0, True=@Multiply_2.Result, Condition=RigVMVariableNode() {variable=isaiming}.Value) {template="DISPATCH_RigVMDispatch_If(in Condition,in True,in False,out Result)"}.Result
   )
   > &RigUnit_ModifyTransforms(
    ItemToModify="((Item=(Type=Bone,Name=\"base_hair_R_01\"),Transform=(Rotation=(X=-0.1773,Y=-0.0362,Z=-0.2063,W=-0.9616),Translation=(X=0.6989,Y=-0.0954,Z=0.9939),Scale3D=(X=1,Y=1,Z=1))),(Item=(Type=Bone,Name=\"base_hair_C_01\"),Transform=(Rotation=(X=-0.1722,Y=0.0227,Z=0.128 ...(truncated)"
    WeightMinimum=0
    Weight=@AlphaInterp.Result
   )
   n19 = &RigVMFunction_MathQuaternionToEuler(
    RotationOrder=YZX
    Value=&RigUnit_GetRelativeTransformForItem(Child=(Type=Bone,Name="clavicle_r"), Parent=(Type=Bone,Name="head"), bChildInitial=False, bParentInitial=False).RelativeTransform.Rotation
   )
   n20 = &RigVMFunction_AlphaInterp(
    InterpSpeedDecreasing=5
    InterpSpeedIncreasing=10
    Scale=1
    bClampResult=false
    bMapRange=false
    Value=RigVMDispatchNode(
     False=0
     True=&RigVMFunction_MathDoubleRemap(SourceMaximum=-20, SourceMinimum=-47, TargetMaximum=0, TargetMinimum=1, Value=@n19.Result.Z).Result
     Condition=RigVMVariableNode() {variable=isaiming}.Value
    ) {template="DISPATCH_RigVMDispatch_If(in Condition,in True,in False,out Result)"}.Result
   )
   > &RigUnit_ModifyTransforms(
    ItemToModify="((Item=(Type=Bone,Name=\"base_hair_R_01\"),Transform=(Rotation=(X=0.007,Y=0.0427,Z=-0.2081,W=0.9771),Translation=(X=0,Y=-2.9802,Z=0.8778),Scale3D=(X=1,Y=1,Z=1))),(Item=(Type=Bone,Name=\"base_hair_L_01\"),Transform=(Rotation=(X=0.0094,Y=-0.0426,Z=-0.2162,W=0.97 ...(truncated)"
    WeightMinimum=0
    Weight=@n20.Result
   )
   &RigUnit_ModifyTransforms(
    ItemToModify="((Item=(Type=Bone,Name=\"base_hair_C_01\"),Transform=(Rotation=(X=0.0234,Y=0.0368,Z=-0.5368,W=0.8426),Translation=(X=-0.0186,Y=-3.7384,Z=0),Scale3D=(X=0.5,Y=1,Z=1))),(Item=(Type=Bone,Name=\"base_hair_L_01\"),Transform=(Rotation=(X=0,Y=0,Z=-0.2588,W=0.9659),Tra ...(truncated)"
    WeightMinimum=0
    Weight=@n16.Result
    ExecutePin=@RigVMFunction_ControlFlowBranch.True
   )
   n21 = &RigUnit_GetRelativeTransformForItem(Child=(Type=Bone,Name="clavicle_r"), Parent=(Type=Null,Name="clavicle_r_null"), bChildInitial=False, bParentInitial=False)
   n22 = &RigUnit_GetRelativeTransformForItem(Child=(Type=Bone,Name="head"), Parent=(Type=Null,Name="head_null"), bChildInitial=False, bParentInitial=False)
   Multiply_3 = RigVMAggregateNode(
    A=1
    C=&RigVMFunction_MathDoubleRemap(
     SourceMaximum=-14
     SourceMinimum=0
     TargetMaximum=1
     TargetMinimum=0
     bClamp=False
     Value=&RigVMFunction_MathQuaternionToEuler(RotationOrder=ZYX, Value=@n21.RelativeTransform.Rotation).Result.Y
    ).Result
    B=&RigVMFunction_MathDoubleRemap(
     SourceMaximum=-20
     SourceMinimum=5
     TargetMaximum=1
     TargetMinimum=0
     Value=&RigVMFunction_MathQuaternionToEuler(RotationOrder=ZYX, Value=@n22.RelativeTransform.Rotation).Result.Z
    ).Result
   )
   n23 = &RigVMFunction_AlphaInterp(
    InterpSpeedDecreasing=5
    InterpSpeedIncreasing=10
    Scale=1
    bClampResult=true
    bMapRange=false
    Value=RigVMDispatchNode(False=0, True=@Multiply_3.Result, Condition=RigVMVariableNode() {variable=isaiming}.Value) {template="DISPATCH_RigVMDispatch_If(in Condition,in True,in False,out Result)"}.Result
   )
   > n24 = &RigUnit_ModifyTransforms(
    ItemToModify="((Item=(Type=Bone,Name=\"base_hair_R_01\"),Transform=(Rotation=(X=-0.0449,Y=0.0842,Z=-0.2549,W=-0.9623),Translation=(X=0,Y=0,Z=0),Scale3D=(X=1,Y=1,Z=1))),(Item=(Type=Bone,Name=\"base_hair_C_01\"),Transform=(Rotation=(X=0.1835,Y=-0.1221,Z=-0.052,W=-0.974),Trans ...(truncated)"
    WeightMinimum=0
    Weight=@n23.Result
   )
   n25 = &RigVMFunction_ControlFlowBranch(ExecuteContext=@n24.ExecutePin, Condition=RigVMVariableNode() {variable=isaiming}.Value)
   n26 = &RigUnit_SphericalPoseReader(
    ActiveRegionScaleFactors=(PositiveWidth=1,NegativeWidth=1,PositiveHeight=1,NegativeHeight=1)
    ActiveRegionSize=0.01
    Debug=(bDrawDebug=true,bDraw2D=False,bDrawLocalAxes=False,DebugScale=25,DebugSegments=20,DebugThickness=0.25)
    DriverAxis=(X=0,Y=1,Z=0)
    DriverItem=(Type=Bone,Name="head")
    FalloffRegionScaleFactors=(PositiveWidth=1,NegativeWidth=1,PositiveHeight=1,NegativeHeight=1)
    FalloffSize=0.12
    FlipHeightScaling=False
    FlipWidthScaling=False
    OptionalParentItem=(Type=Bone,Name="None")
    RotationOffset=(X=100,Y=0,Z=-45)
    ExecutePin=@n25.False
   )
   > &RigUnit_ModifyTransforms(
    ItemToModify="((Item=(Type=Bone,Name=\"base_hair_C_01\"),Transform=(Rotation=(X=0,Y=0,Z=0.3827,W=-0.9239),Translation=(X=0,Y=0,Z=0),Scale3D=(X=1,Y=1,Z=1))),(Item=(Type=Bone,Name=\"base_hair_R_01\"),Transform=(Rotation=(X=-0.0226,Y=0.2578,Z=0.0842,W=0.9623),Translation=(X=0, ...(truncated)"
    WeightMinimum=0
    Weight=@n26.OutputParam
   )
   n27 = &RigUnit_SphericalPoseReader(
    ActiveRegionScaleFactors=(PositiveWidth=1,NegativeWidth=1,PositiveHeight=1,NegativeHeight=1)
    ActiveRegionSize=0.01
    Debug=(bDrawDebug=false,bDraw2D=False,bDrawLocalAxes=False,DebugScale=25,DebugSegments=20,DebugThickness=0.25)
    DriverAxis=(X=0,Y=1,Z=0)
    DriverItem=(Type=Bone,Name="head")
    FalloffRegionScaleFactors=(PositiveWidth=1,NegativeWidth=1,PositiveHeight=1,NegativeHeight=1)
    FalloffSize=0.1
    FlipHeightScaling=False
    FlipWidthScaling=False
    OptionalParentItem=(Type=Bone,Name="clavicle_r")
    RotationOffset=(X=12.8104,Y=-92.6518,Z=-4.1505)
    ExecutePin=@n25.True
   )
   > &RigUnit_ModifyTransforms(
    ItemToModify="((Item=(Type=Bone,Name=\"base_hair_R_01\"),Transform=(Rotation=(X=-0.0735,Y=-0.0468,Z=0.0435,W=0.9952),Translation=(X=0,Y=0,Z=0),Scale3D=(X=1,Y=1,Z=1))),(Item=(Type=Bone,Name=\"base_hair_C_01\"),Transform=(Rotation=(X=0,Y=0,Z=-0.3827,W=0.9239),Translation=(X=- ...(truncated)"
    WeightMinimum=0
    Weight=@n27.OutputParam
   )
   n28 = &RigVMFunction_MathQuaternionToEuler(
    RotationOrder=XZY
    Value=&RigUnit_GetRelativeTransformForItem(Child=(Type=Bone,Name="head"), Parent=(Type=Bone,Name="spine_05"), bChildInitial=False, bParentInitial=False).RelativeTransform.Rotation
   )
   &RigUnit_ModifyTransforms(
    ItemToModify="((Item=(Type=Bone,Name=\"base_hair_L_01\"),Transform=(Rotation=(X=-0.1574,Y=0.0734,Z=0.0858,W=-0.9811),Translation=(X=0,Y=0,Z=0),Scale3D=(X=1,Y=1,Z=1))),(Item=(Type=Bone,Name=\"base_hair_C_01\"),Transform=(Rotation=(X=0,Y=0,Z=0,W=1),Translation=(X=0,Y=-0.8424, ...(truncated)"
    WeightMinimum=0
    Weight=&RigVMFunction_MathDoubleRemap(SourceMaximum=-60, SourceMinimum=-20, TargetMaximum=1, TargetMinimum=0, Value=@n28.Result.X).Result
    ExecutePin=@RigVMFunction_Sequence.B
   )
   > &RigUnit_ModifyTransforms(
    ItemToModify="((Item=(Type=Bone,Name=\"base_hair_R_01\"),Transform=(Rotation=(X=0.0262,Y=-0.0435,Z=-0.0831,W=0.9952),Translation=(X=0.9697,Y=0,Z=0.3529),Scale3D=(X=1,Y=1,Z=1))),(Item=(Type=Bone,Name=\"base_hair_FR_00\"),Transform=(Rotation=(X=0,Y=0,Z=0,W=1),Translation=(X=0 ...(truncated)"
    WeightMinimum=0
    Weight=&RigVMFunction_MathDoubleRemap(SourceMaximum=60, SourceMinimum=0, TargetMaximum=1, TargetMinimum=0, Value=@n28.Result.X).Result
   )
   RigVMDispatchNode(True=0.3, False=@n16.Result, Condition=RigVMVariableNode() {variable=ismoving2D}.Value) {template="DISPATCH_RigVMDispatch_If(in Condition,in True,in False,out Result)"}
   RigVMFunction_MathVectorMake = &RigVMFunction_MathVectorMake(X=-141.5, Y=-138.5, Z=-9.3)
   RigVMFunction_MathVectorDistance = &RigVMFunction_MathVectorDistance(
    A=&RigVMFunction_MathQuaternionToEuler(
     RotationOrder=YZX
     Value=&RigUnit_GetRelativeTransformForItem(Child=(Type=Bone,Name="clavicle_r"), Parent=(Type=Bone,Name="head"), bChildInitial=False, bParentInitial=False).RelativeTransform.Rotation
    ).Result
    B=@RigVMFunction_MathVectorMake.Result
   )
   &RigUnit_ModifyTransforms(
    ExecutePin=()
    ItemToModify="((Item=(Type=Bone,Name=\"base_hair_R_01\"),Transform=(Rotation=(X=0,Y=0,Z=-0.1305,W=0.9914),Translation=(X=4.9473,Y=0,Z=0),Scale3D=(X=1,Y=1,Z=1))),(Item=(Type=Bone,Name=\"None\"),Transform=(Rotation=(X=0,Y=0,Z=0,W=1),Translation=(X=0,Y=0,Z=0),Scale3D=(X=1,Y=1, ...(truncated)"
    WeightMinimum=0
    Weight=&RigVMFunction_MathDoubleRemap(SourceMaximum=27, SourceMinimum=47, TargetMaximum=1, TargetMinimum=0, Value=@RigVMFunction_MathVectorDistance.Result).Result
   )
   &RigUnit_ModifyTransforms(
    ExecutePin=()
    ItemToModify=((Item=(Type=Bone,Name="base_hair_R_01"),Transform=(Rotation=(X=-0.2091,Y=0.2527,Z=0.056,W=0.943),Translation=(X=-1.3699,Y=0.0286,Z=2.3038),Scale3D=(X=1,Y=1,Z=1))))
    WeightMinimum=0
    Weight=@n13.Result
   )
   &RigVMFunction_AnimEvalRichCurve(
    Curve="(EditorCurveData=(Keys=((InterpMode=RCIM_Linear,TangentMode=RCTM_Auto,TangentWeightMode=RCTWM_WeightedNone,Time=0,Value=0,ArriveTangent=0,ArriveTangentWeight=0,LeaveTangent=0,LeaveTangentWeight=0),(InterpMode=RCIM_Linear,TangentMode=RCTM_Auto,TangentWeight ...(truncated)"
    SourceMinimum=0
    TargetMinimum=0
    Value=@Multiply.Result
   )
   &RigVMFunction_AnimEvalRichCurve(
    Curve="(EditorCurveData=(Keys=((InterpMode=RCIM_Linear,TangentMode=RCTM_Auto,TangentWeightMode=RCTWM_WeightedNone,Time=0,Value=0,ArriveTangent=0,ArriveTangentWeight=0,LeaveTangent=0,LeaveTangentWeight=0),(InterpMode=RCIM_Linear,TangentMode=RCTM_Auto,TangentWeight ...(truncated)"
    SourceMinimum=0
    TargetMinimum=0
    Value=@n15.Result
   )
   RigVMVariableNode() {variable=aimalpha}
   RigVMVariableNode() {variable=headrotationZ}
   RigVMVariableNode() {variable=headrotationZ}
  ]}
]
hierarchy [
 root {loc=<0 0 0>, rot=<0 0 0 1>}
 pelvis {parent=root, loc=<0 -0.5301 98.0571>, rot=<0.0127 0.707 -0.0127 -0.707>}
 spine_01 {parent=pelvis, loc=<6.1208 -0.9328 0>, rot=<0 0 0.034 -0.9994>}
 spine_02 {parent=spine_01, loc=<6.5997 -0.2095 0>, rot=<0 0 0.0238 -0.9997>}
 spine_03 {parent=spine_02, loc=<6.4914 -0.1237 0>, rot=<0 0 0.0149 -0.9999>}
 spine_04 {parent=spine_03, loc=<6.9598 0.0794 0>, rot=<0 0 -0.0132 -0.9999>}
 spine_05 {parent=spine_04, loc=<6.9987 0 0>, rot=<0 0 -0.0015 -1>}
 neck_01 {parent=spine_05, loc=<11.3138 0.1529 0>, rot=<0 0 -0.2003 -0.9797>}
 neck_02 {parent=neck_01, loc=<4.5627 0 0>, rot=<0 0 0.0504 -0.9987>}
 head {parent=neck_02, loc=<5.3914 0 0>, rot=<0 0 0.115 -0.9934>}
 base_hair_C_01 {parent=head, loc=<5.8126 -5.4029 0>, rot=<0 0 1 0>}
 dyn_hair_C_01 {parent=base_hair_C_01, loc=<0 0 0>, rot=<0 0 0 1>}
 dyn_hair_C_02 {parent=dyn_hair_C_01, loc=<5 0 0>, rot=<0 0 0 1>}
 dyn_hair_C_03 {parent=dyn_hair_C_02, loc=<5 0 0>, rot=<0 0 0 1>}
 base_hair_L_01 {parent=head, loc=<5.7962 -3.4173 -5.7321>, rot=<0 0 1 0>}
 dyn_hair_L_01 {parent=base_hair_L_01, loc=<0 0 0>, rot=<0 0 0 1>}
 dyn_hair_L_02 {parent=dyn_hair_L_01, loc=<5 0 0>, rot=<0 0 0 1>}
 dyn_hair_L_03 {parent=dyn_hair_L_02, loc=<5 0 0>, rot=<0 0 0 1>}
 base_hair_R_01 {parent=head, loc=<5.794 -3.1847 5.629>, rot=<0 0 0 1>}
 dyn_hair_R_01 {parent=base_hair_R_01, loc=<0 0 0>, rot=<0 0 0 1>}
 dyn_hair_R_02 {parent=dyn_hair_R_01, loc=<-4.9991 0.0002 0.0001>, rot=<0 0 0 1>}
 dyn_hair_R_03 {parent=dyn_hair_R_02, loc=<-5.0001 0 0>, rot=<0 0 0 1>}
 head_null {parent=neck_02, type=Null, loc=<5.3914 0 0>, rot=<0 0 0.115 -0.9934>}
 clavicle_l {parent=spine_05, loc=<5.4831 -0.1053 -3.2946>, rot=<0.0465 -0.7289 -0.009 -0.6829>}
 upperarm_l {parent=clavicle_l, loc=<9.8658 -1.2136 -0.4105>, rot=<0.0164 -0.3245 -0.0269 -0.9454>}
 clavicle_r {parent=spine_05, loc=<5.4831 -0.1053 3.2946>, rot=<0.7289 0.0465 -0.6829 0.009>}
 upperarm_r {parent=clavicle_r, loc=<-9.8658 1.2136 0.4105>, rot=<0.0164 -0.3245 -0.0269 -0.9454>}
 clavicle_l_null {parent=spine_05, type=Null, loc=<5.483 -0.1053 -3.2946>, rot=<0.0465 -0.7289 -0.009 -0.6829>}
 clavicle_r_null {parent=spine_05, type=Null, loc=<5.483 -0.1053 3.2946>, rot=<0.7289 0.0465 -0.6829 0.009>}
]
variables [
 aimalpha {type=double, default=0}
 alphacrouch {type=double, default=0}
 hasRBAN {type=bool, default=False}
 headrotationZ {type=double, default=0}
 isaiming {type=bool, default=False}
 iscrouching {type=bool, default=False}
 ismoving2D {type=bool, default=False}
]
