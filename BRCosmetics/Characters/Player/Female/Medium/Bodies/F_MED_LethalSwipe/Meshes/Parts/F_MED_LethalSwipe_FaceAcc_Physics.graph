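# Physics asset for the face accessory part of the LethalSwipe medium female character, covering three hair chains and the upper body around them.
# Center, left and right chains each link three capsule bodies to a base bone locked to the head, with zero twist limits and swing limits that widen toward the tips.
# Head, neck, upper spine and clavicle capsules are angularly locked together, while upper arm capsules rotate freely.
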PhysicsAsset /BRCosmetics/Characters/Player/Female/Medium/Bodies/F_MED_LethalSwipe/Meshes/Parts/F_MED_LethalSwipe_FaceAcc_Physics
bodies {
 base_hair_C_01 [{shapeType=Capsule, length=0.505, radius=0.505, shapeName=base_hair_C_01_capsule, rotation=<rot 0 -90 0>}]
 base_hair_L_01 [{shapeType=Capsule, length=0.505, radius=0.505, shapeName=base_hair_L_01_capsule, rotation=<rot 0 -90 0>}]
 base_hair_R_01 [{shapeType=Capsule, length=0.505, radius=0.505, shapeName=base_hair_R_01_capsule, rotation=<rot 0 -90 0>}]
 clavicle_l [{shapeType=Capsule, length=13.5, radius=7, shapeName=clavicle_l_capsule, center=<5.46 -1.1196 -1.4243>, rotation=<rot 75 -0.0009 179.9991>}]
 clavicle_r [{shapeType=Capsule, length=13.5, radius=7, shapeName=clavicle_r_capsule, center=<-5.46 1.1196 1.4243>, rotation=<rot 75 -0.0009 -0.0009>}]
 dyn_hair_C_01 [{shapeType=Capsule, length=7, radius=1, shapeName=dyn_hair_C_01_capsule, center=<2 2.1764 0>}]
 dyn_hair_C_02 [{shapeType=Capsule, length=7.5, radius=2.5, shapeName=dyn_hair_C_02_capsule, center=<3.0529 3.4 0>}]
 dyn_hair_C_03 [{shapeType=Capsule, length=9, radius=2.5, shapeName=dyn_hair_C_03_capsule, center=<3.1102 3.9496 0>}]
 dyn_hair_L_01 [{shapeType=Capsule, length=7, radius=1, shapeName=dyn_hair_L_01_capsule, center=<2 0 -0.6>, rotation=<rot 0.0001 0.0001 70>}]
 dyn_hair_L_02 [{shapeType=Capsule, length=7, radius=1, shapeName=dyn_hair_L_02_capsule, center=<2 0 -0.8>, rotation=<rot 0 0.0002 70>}]
 dyn_hair_L_03 [{shapeType=Capsule, length=6, radius=2, shapeName=dyn_hair_L_03_capsule, center=<2.233 0.8551 -2.5994>, rotation=<rot 0 0 -110>}]
 dyn_hair_R_01 [{shapeType=Capsule, length=7, radius=1, shapeName=dyn_hair_R_01_capsule, center=<-2 0 0.75>, rotation=<rot 0 -0.0001 70>}]
 dyn_hair_R_02 [{shapeType=Capsule, length=7, radius=1, shapeName=dyn_hair_R_02_capsule, center=<-2 0 1>, rotation=<rot -0.0001 -0.0001 70>}]
 dyn_hair_R_03 [{shapeType=Capsule, length=7.5, radius=1.5, shapeName=dyn_hair_R_03_capsule, center=<-3.3109 -0.2154 1.6036>, rotation=<rot 0 0 70>}]
 head [{shapeType=Capsule, length=3.7486, radius=6.2894, shapeName=head_capsule, rotation=<rot 90 0 0>}]
 neck_01 [{shapeType=Capsule, length=0.505, radius=0.505, shapeName=neck_01_capsule, rotation=<rot 0 -90 0>}]
 neck_02 [{shapeType=Capsule, length=0.505, radius=0.505, shapeName=neck_02_capsule, rotation=<rot 0 -90 0>}]
 spine_05 [{shapeType=Capsule, length=13.5, radius=10.505, shapeName=spine_05_capsule, center=<1.3306 0 0>, rotation=<rot 0 -90 0>}]
 upperarm_l [{shapeType=Capsule, length=19, radius=7.005, shapeName=upperarm_l_capsule, center=<14.0963 0 0>, rotation=<rot 90 180 0>}]
 upperarm_r [{shapeType=Capsule, length=19, radius=7.005, shapeName=upperarm_r_capsule, center=<-14.0963 0 0>, rotation=<rot 90 0 0>}]
}
constraints [
 {bone1Name=dyn_hair_C_01, bone2Name=base_hair_C_01, swing1LimitDegrees=5, swing2LimitDegrees=5, twistLimitDegrees=0}
 {bone1Name=dyn_hair_C_02, bone2Name=dyn_hair_C_01, swing1LimitDegrees=10, swing2LimitDegrees=8, twistLimitDegrees=0}
 {bone1Name=dyn_hair_C_03, bone2Name=dyn_hair_C_02, swing1LimitDegrees=25, swing2LimitDegrees=10, twistLimitDegrees=0}
 {bone1Name=dyn_hair_L_01, bone2Name=base_hair_L_01, swing1LimitDegrees=5, swing2LimitDegrees=5, twistLimitDegrees=0}
 {bone1Name=dyn_hair_L_02, bone2Name=dyn_hair_L_01, swing1LimitDegrees=10, swing2LimitDegrees=8, twistLimitDegrees=0}
 {bone1Name=dyn_hair_L_03, bone2Name=dyn_hair_L_02, swing1LimitDegrees=15, swing2LimitDegrees=10, twistLimitDegrees=0}
 {bone1Name=dyn_hair_R_01, bone2Name=base_hair_R_01, swing1LimitDegrees=5, swing2LimitDegrees=5, twistLimitDegrees=0}
 {bone1Name=dyn_hair_R_02, bone2Name=dyn_hair_R_01, swing1LimitDegrees=10, swing2LimitDegrees=8, twistLimitDegrees=0}
 {bone1Name=dyn_hair_R_03, bone2Name=dyn_hair_R_02, swing1LimitDegrees=45, swing2LimitDegrees=15, twistLimitDegrees=0}
 {
  bone1Name=base_hair_C_01
  bone2Name=head
  swing1LimitDegrees=45
  swing1Motion=Locked
  swing2LimitDegrees=45
  swing2Motion=Locked
  twistLimitDegrees=45
  twistMotion=Locked
 }
 {
  bone1Name=base_hair_L_01
  bone2Name=head
  swing1LimitDegrees=45
  swing1Motion=Locked
  swing2LimitDegrees=45
  swing2Motion=Locked
  twistLimitDegrees=45
  twistMotion=Locked
 }
 {
  bone1Name=base_hair_R_01
  bone2Name=head
  swing1LimitDegrees=45
  swing1Motion=Locked
  swing2LimitDegrees=45
  swing2Motion=Locked
  twistLimitDegrees=45
  twistMotion=Locked
 }
 {
  bone1Name=neck_02
  bone2Name=neck_01
  swing1LimitDegrees=45
  swing1Motion=Locked
  swing2LimitDegrees=45
  swing2Motion=Locked
  twistLimitDegrees=45
  twistMotion=Locked
 }
 {
  bone1Name=head
  bone2Name=neck_02
  swing1LimitDegrees=45
  swing1Motion=Locked
  swing2LimitDegrees=45
  swing2Motion=Locked
  twistLimitDegrees=45
  twistMotion=Locked
 }
 {
  bone1Name=neck_01
  bone2Name=spine_05
  swing1LimitDegrees=45
  swing1Motion=Locked
  swing2LimitDegrees=45
  swing2Motion=Locked
  twistLimitDegrees=45
  twistMotion=Locked
 }
 {
  bone1Name=clavicle_l
  bone2Name=spine_05
  swing1LimitDegrees=45
  swing1Motion=Locked
  swing2LimitDegrees=45
  swing2Motion=Locked
  twistLimitDegrees=45
  twistMotion=Locked
 }
 {
  bone1Name=clavicle_r
  bone2Name=spine_05
  swing1LimitDegrees=45
  swing1Motion=Locked
  swing2LimitDegrees=45
  swing2Motion=Locked
  twistLimitDegrees=45
  twistMotion=Locked
 }
 {
  bone1Name=upperarm_l
  bone2Name=clavicle_l
  swing1LimitDegrees=45
  swing1Motion=Free
  swing2LimitDegrees=45
  swing2Motion=Free
  twistLimitDegrees=45
  twistMotion=Free
 }
 {
  bone1Name=upperarm_r
  bone2Name=clavicle_r
  swing1LimitDegrees=45
  swing1Motion=Free
  swing2LimitDegrees=45
  swing2Motion=Free
  twistLimitDegrees=45
  twistMotion=Free
 }
]
