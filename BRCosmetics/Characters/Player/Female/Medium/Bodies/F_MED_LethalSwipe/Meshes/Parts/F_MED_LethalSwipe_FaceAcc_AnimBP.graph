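# Anim Blueprint for the face accessory part of the female LethalSwipe character cosmetic.
# Copies the attached parent's pose, then chains a Control Rig fed crouch, aim and movement state, an additive head-bone scale for LOD 2–3, and head-relative RigidBody simulation.
# Each update computes gravity override and rigid-body settings from the character's parameter asset and refreshes the LOD flag; targeting events toggle the aim inputs.
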AnimBlueprint /BRCosmetics/Characters/Player/Female/Medium/Bodies/F_MED_LethalSwipe/Meshes/Parts/F_MED_LethalSwipe_FaceAcc_AnimBP
target_skeleton=&/F_MED_LethalSwipe_FaceAcc_Skeleton
graphs [
 AnimGraph {nodes [
   Output Pose(Result=Use cached pose 'RigidBody'()) {layerGroup=None}
   n1 = Control Rig!(
    Source=Copy Pose From Mesh() {bUseAttachedParent=true}
    Alpha
    iscrouching:bool=Get bIsCrouching()
    isaiming:bool=Get Aim() -> (Aim:bool)
    aimalpha:double=Get AimAlpha() -> (AimAlpha:double)
    ismoving2D:bool=Get bIsMoving2D()
    hasRBAN:bool=Get bRigidBodyEnabled()
   ) -> (Pose) {controlRigAssetReference {blueprintRigClass=&/F_MED_LethalSwipe_FaceAcc_CtrlRig.F_MED_LethalSwipe_FaceAcc_CtrlRig_C}}
   Save cached pose 'ControlRig'(Pose=@n1) {cacheName=ControlRig}
   n2 = "Transform (Modify) Bone - Bone: head"!(
    ComponentPose=Local To Component(LocalPose=Use cached pose 'ControlRig'())
    bAlphaBoolEnabled:bool=Get LOD23() -> (LOD23:bool)
    Translation="0,0,0"
    Rotation="0,0,0"
    Scale="1.06,1.06,1.06"
   ) -> (Pose) {alphaInputType=Bool, scaleMode=BMM_Additive, scaleSpace=BCS_BoneSpace, boneToModify {boneName=head}, scale=<1.06 1.06 1.06>}
   Save cached pose 'Transforms'(Pose=Component To Local(ComponentPose=@n2)) {cacheName=Transforms}
   n3 = RigidBody!(ComponentPose=Local To Component(LocalPose=Use cached pose 'Transforms'()), Alpha, OverrideWorldGravity:Vector="0,0,0", ExternalForce) -> (Pose) {
    bOverrideWorldGravity=true
    overridePhysicsAsset=&/F_MED_LethalSwipe_FaceAcc_Physics
    simulationSpace=BaseBoneSpace
    baseBoneRef {boneName=head}
    simSpaceSettings {
     maxAngularAcceleration=40
     maxAngularVelocity=8
     maxLinearAcceleration=4000
     maxLinearVelocity=800
     velocityScaleZ=0.5
     worldAlpha=0.5
     externalLinearDragV=<0.3 0.2 0.2>
     externalLinearVelocity=<0.3 0.2 0.2>
    }
   }
   Save cached pose 'RigidBody'(Pose=Component To Local(ComponentPose=@n3)) {cacheName=RigidBody}
  ]}
 EventGraph {nodes [
   Event BlueprintUpdateAnimation()
   > n1 = "Gravity Override State Function (Parameter Asset) (Nativized)"(
    CharacterPartAnimInstance=Self-Reference()
    GameTimeInSeconds=GetGameTimeInSeconds()
    Parameters=/BRCosmetics/Characters/Player/Female/Medium/Bodies/F_MED_LethalSwipe/Meshes/F_MED_LethalSwipe_Parameters
   )
   > Set GravityOverride(GravityOverride=@n1)
   > Set RBAN Settings Out(RBAN Settings Out=@n1.OutRigidBodyAnimNodeParameters)
   n2 = Get CurrentLOD()
   n3 = Get CurrentLOD()
   > Set LOD23(LOD23:bool=OR Boolean(A="Equal (Integer)"(A=@n3, B=2), B="Equal (Integer)"(A=@n2, B=3))) -> (Output_Get:bool)
   Event OnBeginTargeting() {is=title K2Node_Event}
   > Set Aim(Aim:bool=true) -> (Output_Get:bool)
   > Set AimAlpha(AimAlpha:double=1) -> (Output_Get:double)
   Event OnEndTargeting() {is=title K2Node_Event}
   > Set Aim(Aim=false)
   > Set AimAlpha(AimAlpha=0)
  ]}
]
variables [
 Aim {type=boolean}
 AimAlpha {type=number}
 GravityOverride {type=Vector}
 LOD23 {type=boolean}
 RBAN Settings Out {type=FortRigidBodyAnimNodeParameters}
]
